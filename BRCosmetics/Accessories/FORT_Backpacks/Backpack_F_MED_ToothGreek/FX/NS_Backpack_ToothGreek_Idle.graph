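# Idle Niagara effect for the ToothGreek cosmetic backpack, adding blue glows and drifting dust at its sockets.
# Glow_and_Light bursts four sprites that do not expire with lifetime, using spawn group to choose each one's socket and size; despite the name, it has only a sprite renderer.
# Dust continuously spawns sprites in a small cylinder at socket 3, with random velocity, drag, SubUV animation and size and color curves that fade them out.
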NiagaraSystem /BRCosmetics/Accessories/FORT_Backpacks/Backpack_F_MED_ToothGreek/FX/NS_Backpack_ToothGreek_Idle
dependencies {
 usedDataInterfaces [
  &NiagaraDataInterfaceSpriteRendererInfo
  &NiagaraDataInterfaceCurve
  &NiagaraDataInterfaceVector2DCurve
  &NiagaraDataInterfaceColorCurve
  &NiagaraDataInterfaceSocketReader
 ]
 usedDynamicInputs [&FloatFromCurve]
 usedModules [
  &self:SystemSpawnScript.SystemScriptSource.SystemScriptGraph.NiagaraNodeAssignment_0.SetVariables_#1
  &SystemState
 ]
}
emitters {
 Dust {
  emitterUpdateScript [
   &EmitterState(Life Cycle Mode=::NewEnumerator1, Loop Duration=1, UseLoopDelay=false)
   &SpawnRate(SpawnRate=4.5, Use Spawn Probability=false)
  ]
  particleSpawnScript [
   &InitializeParticle(
    Color Maximum=<rgba 0 0.503 2 0.6957>
    Color Minimum=<rgba 0 0.4511 1 0.9848>
    Color Mode=::NewEnumerator2
    Lifetime Min=1.5
    Lifetime Mode=::NewEnumerator1
    Sprite Rotation Mode=::NewEnumerator2
    Sprite Size Mode=::NewEnumerator3
    Uniform Sprite Size=15
    UsePositionOffset=false
    Write Material Random=true
   )
   &AddVelocity(Velocity=&RandomVector(Vector Scale=1), Velocity Speed Scale=8)
   &SocketLocation(SamplePosOffset=false, Socket Index=3, Socket Reader=$System.Socket Reader, Socket Sampling Mode=::NewEnumerator3, VelocitySpeedLimit=false)
   &ShapeLocation(
    Cylinder Distribution=::NewEnumerator2
    Cylinder Height=1
    Cylinder Radius=3
    MaskSpawnGroup=false
    Rotation Angle Type:ENiagara_AngleInput=::NewEnumerator0
    Rotation Axis=<0 1 0>
    Rotation Mode=::NewEnumerator1
    Shape Primitive=::NewEnumerator1
    U Position=&RandomRangeFloat(Minimum=0, Maximum=1)
    UseOwnerScale=true
    V Position=&RandomRangeFloat(Minimum=0, Maximum=1)
   ) {module=ShapeLocation001}
   &CameraOffset(Camera Offset Amount=2)
  ]
  particleUpdateScript [
   &ParticleState()
   &Drag(Use Linear Drag=true, Use Rotational Drag=true)
   &SubUVAnimation(
    Random Start Frame=true
    "Sprite Renderer "=DataInterface() {propertyValues {SpriteRenderer=&self:Dust_0.NiagaraSpriteRendererProperties_2}}
    SubUV Animation Mode:ENiagara_SubUVLookupModeV2=::NewEnumerator1
    SubUV Curve=DataInterface() {propertyValues {Curve {
       defaultValue=0
       keys [
        {arriveTangent=0.412, interpMode=RCIM_Cubic, leaveTangent=0.412, tangentMode=RCTM_User, time=0, value=0}
        {arriveTangent=0.4055, interpMode=RCIM_Cubic, leaveTangent=0.4055, tangentMode=RCTM_User, time=1, value=0.4}
       ]
      }}}
    UseEndFrame=false
    UseStartFrame=false
   )
   &ScaleSpriteSize(Uniform Curve Sprite Scale=DataInterface() {propertyValues {Curve {
       defaultValue=0.3
       keys [
        {arriveTangent=3.347, interpMode=RCIM_Cubic, leaveTangent=3.347, tangentMode=RCTM_User, time=0, value=0.3997}
        {arriveTangent=1.1068, interpMode=RCIM_Cubic, leaveTangent=1.1068, time=0.1583, value=0.8549}
        {arriveTangent=0.0224, interpMode=RCIM_Cubic, leaveTangent=0.0224, tangentMode=RCTM_User, time=0.5424, value=1}
       ]
      }}})
   &ScaleColor(
    Linear Color Curve=DataInterface() {propertyValues {
      AlphaCurve {keys [<at 0 0>,<at 0.1624 0.3501>,<at 0.376 1.3>,<at 1 0>]}
      BlueCurve {keys [<at 0.2153 1>,<at 0.6433 0.0522>,<at 1 0>]}
      GreenCurve {keys [<at 0.2153 1>,<at 0.6433 0.0481>,<at 1 0>]}
      RedCurve {keys [<at 0.2153 1>,<at 0.6433 0.0481>,<at 1 0>]}
     }}
    Scale Mode=2
   )
   &SolveForcesAndVelocity(Clamp Velocity=false, Limit Acceleration=false)
  ]
  rendererClasses [&NiagaraSpriteRendererProperties]
  renderers [{rendererClass=&NiagaraSpriteRendererProperties}]
 }
 Glow_and_Light {
  emitterUpdateScript [
   &EmitterState(Inactive Response=::NewEnumerator2, Life Cycle Mode=::NewEnumerator1, Loop Behavior=::NewEnumerator1, Loop Duration=2, UseLoopDelay=false)
   &SpawnBurst_Instantaneous(Spawn Count=4, Use Spawn Probability=false, UseLoopCountLimit=true)
  ]
  particleSpawnScript [
   &FN_SetSpawnGroup()
   &InitializeParticle(
    Color=<rgba 0 0.3502 1 0.5>
    Sprite Size Mode=::NewEnumerator3
    Uniform Sprite Size=&Multiply_Float(
     A=6.7604
     B=&MergeHelper_FloatSwitch(0=1, 1=0.53, 2=0.4, 3=4.28, Index=$Particles.SpawnGroup, Num=::NewEnumerator2)
    )
    UsePositionOffset=false
    Write Material Random=true
   )
   &CameraOffset(Camera Offset Amount=5)
   &SocketLocation(
    SamplePosOffset=false
    Socket Index=$Particles.SpawnGroup
    Socket Reader=$System.Socket Reader
    Socket Sampling Mode=::NewEnumerator3
    VelocitySpeedLimit=false
   )
  ]
  particleUpdateScript [&ParticleState(Kill Particles When Lifetime Has Elapsed=false)]
  rendererClasses [&NiagaraSpriteRendererProperties]
  renderers [{rendererClass=&NiagaraSpriteRendererProperties}]
 }
}
summary {
 systemName=NS_Backpack_ToothGreek_Idle
 emitters [
  {emitterName=Glow_and_Light, rendererClasses [&NiagaraSpriteRendererProperties]}
  {emitterName=Dust, rendererClasses [&NiagaraSpriteRendererProperties]}
 ]
 userVariables [
  User.bIsFrontEnd {type=NiagaraBool}
  User.bIsFrontEndPreview {type=NiagaraBool}
  User.VariantIndex {type=NiagaraInt32}
 ]
}
